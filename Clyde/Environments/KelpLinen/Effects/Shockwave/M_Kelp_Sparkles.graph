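# Translucent unlit sprite material for twinkling, multicolored sparkle particles in the KelpLinen shockwave effect.
# Per-particle randoms choose among three sparkle, spark and starburst masks, vary sine twinkle speed and shift the particle color's hue, blended in by a parameter.
# ES2 and ES3_1 use a cheaper hue-shifted starburst without twinkle for emissive.
# The chosen mask is also multiplied by particle alpha and a KelpLinen location fade.
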Material /Clyde/Environments/KelpLinen/Effects/Shockwave/M_Kelp_Sparkles
graph {
 properties {
  bUsedWithNiagaraSprites=true
  bUsedWithParticleSprites=true
  blendMode=BLEND_Translucent
  pixelDepthOffsetMode=PDOM_Legacy
  shadingModel=MSM_Unlit
  usedShadingModels=Unlit
 }
 nodes [
  n1 = TextureSample() {texture=/Game/Effects/Fort_Effects/Textures/Flares/T_Starburst}
  n2 = ParticleRandom()
  n3 = MaterialFunctionCall(
   A=TextureSample() {texture=/Game/Effects/Fort_Effects/Textures/Sparkle/T_Sparkle_01}.R
   B=TextureSample() {texture=/Game/CharClasses/Outlander/Gadgets/AntiMaterialCharge/Materials/T_AntiMaterialCharge_Sparks_01}.G
   C=@n1.G
   Alpha=Ceil(None=Multiply(A=@n2) {constB=3})
  ) {materialFunction=&3ColorBlend}
  n4 = ParticleColor()
  n5 = ParticleRandom()
  n6 = Multiply(A=@n5) {constB=0.8}
  n7 = MaterialFunctionCall(Speed=Multiply(A=@n5, B=ScalarParameter() {defaultValue=3, parameterName=TwinkleSpeed})) {materialFunction=&DebugTimeSine}
  n8 = Multiply(A=Clamp(None=Multiply(A=@n7.Result) {constB=5}) {maxDefault=5, minDefault=0.25}, B=@n4.RGB)
  n9 = Multiply(
   A=@n3.Result
   B=LinearInterpolate(
    A=@n4.RGB
    B=MaterialFunctionCall(Hue Shift Percentage=@n6, Texture=@n8) {materialFunction=&HueShift}.Result
    Alpha=ScalarParameter() {defaultValue=1, parameterName=Use Particle Random Color}
   )
  )
  n10 = Multiply(
   A=MaterialFunctionCall(Hue Shift Percentage=@n5, Texture=@n4.RGB) {materialFunction=&HueShift}.Result
   B=@n1.G
  )
  MaterialFunctionCall(Emissive=FeatureLevelSwitch(Default=@n9, ES2=@n10, ES3_1=@n10, SM4_REMOVED=@n9)) {materialFunction=&MF_FX_EmissiveAdjustments}
  Multiply(
   A=Multiply(A=@n3.Result, B=ParticleColor().A)
   B=MaterialFunctionCall() {materialFunction=/ClydeCosmetics/Environments/KelpLinen/MF_KelpLinen_LocationFade}.Output
  )
 ]
}
