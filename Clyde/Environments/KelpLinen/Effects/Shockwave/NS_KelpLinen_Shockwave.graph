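# Niagara system for the KelpLinen environment shockwave: a large wave mesh travelling along the system's wind direction, trailed by cyan-to-purple smoke and sparkles.
# Two near-identical mesh emitters, WaveMesh (GPU) and WaveMesh001, ease out across the water length, then shrink and fade through curves and material parameters.
# GPU sprite emitters read WaveMesh particles to spawn along arcs, gated by wave age.
# Time of day scales smoke alpha; an NPC phase tracker switches sparkle rate.
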NiagaraSystem /Clyde/Environments/KelpLinen/Effects/Shockwave/NS_KelpLinen_Shockwave
dependencies {
 usedDataInterfaces [&NiagaraDataInterfaceCurve, &NiagaraDataInterfaceParticleRead, &NiagaraDataInterfaceColorCurve]
 usedDynamicInputs [
  &NormalizeVector
  &SubtractVectorFromPosition
  &Add_Float
  &DistanceBetweenPositions
 ]
 usedModules [
  &SystemState
  &Timeline
  &self:SystemSpawnScript.SystemScriptSource.SystemScriptGraph.NiagaraNodeAssignment_0.SetVariables_#1
  &self:SystemSpawnScript.SystemScriptSource.SystemScriptGraph.NiagaraNodeAssignment_2.SetVariables_#2
  &self:SystemSpawnScript.SystemScriptSource.SystemScriptGraph.NiagaraNodeAssignment_1.SetVariables_#3
 ]
}
emitters {
 Smoke_V2 {
  simTarget=GPUComputeSim
  emitterUpdateScript [
   &EmitterState(Life Cycle Mode=::NewEnumerator1, Loop Duration=1, UseLoopDelay=false) {module=EmitterLifeCycle}
   &SpawnParticlesFromOtherEmitter(
    Attribute Reader=DataInterface() {propertyValues {EmitterBinding {bindingMode=Other, emitterName=WaveMesh}}}
    Spawn Rate=200
    Spawn Rate Per Particle Cap=200
    Use Maximum Age=false
    Use Minimum Age=false
   )
  ]
  particleSpawnScript [
   &InitializeParticle(
    Color=&RandomRangeLinearColor(Minimum=<rgba 0 0.8192 1 0.1>, Maximum=<rgba 0.6188 0 1 0.2>)
    Lifetime=&RandomRangeFloat(Minimum=2, Maximum=3)
    Sprite Size Mode=::NewEnumerator3
    Uniform Sprite Size=2000
    UsePositionOffset=false
    Write Material Random=true
   ) {module=InitializeParticle001}
   &SampleParticlesFromOtherEmitter(
    Age Sampling:ENiagara_DirectReadSamplingMode=::NewEnumerator1
    Color Sampling=::NewEnumerator1
    Lifetime Sampling:ENiagara_DirectReadSamplingMode=::NewEnumerator1
    Position Sampling:ENiagara_DirectReadSamplingMode=::NewEnumerator1
   )
   /SetVariables(Particles.Random=&RandomRangeFloat(Minimum=0, Maximum=1)) {isSetParameters=true, module=SetVariables_#4}
   &ShapeLocation(
    Additional Quaternion Rotation=false
    Cylinder Distribution=::NewEnumerator2
    Cylinder Height=0
    Cylinder Radius=30000
    MaskSpawnGroup=false
    Offset=&Multiply_VectorByFloat(Vector=$System.WindDirection, Float=-30000)
    Offset Mode=::NewEnumerator0
    Rotation Mode=::NewEnumerator3
    Rotation Quaternion=&FindQuatBetweenTwoVectors(From Vector=$System.WindDirection, To Vector=<1 0 0>)
    Shape Origin=&Add_Vector(A=<0 0 -500>, B=$Output.SampleParticlesFromOtherEmitter.SampledPosition)
    Shape Primitive=::NewEnumerator1
    U Position=&FloatFromCurve(FloatCurve=DataInterface() {propertyValues {Curve {keys [<at 0 0.14>,<at 1 0.36>]}}}, CurveIndex=$Particles.Random)
    UseOwnerScale=true
   ) {module=ShapeLocation001}
   &AddVelocity(Rotation Coordinate Space=::World, Velocity=$System.WindDirection, Velocity Speed Scale=2000)
   &Color(Color=&MultiplyLinearColor(
     Linear Color=&RandomRangeLinearColor(Minimum=<rgba 0 0.8192 1 0.25>, Maximum=<rgba 0.6188 0 1 0.25>)
     Scale Factor=&MakeVector4(
      X=1
      Y=1
      Z=1
      W=&FloatFromCurve(
       FloatCurve=DataInterface() {propertyValues {Curve {keys [<at 16 1>,<at 17 0>]}}}
       CurveIndex=$Output.SampleParticlesFromOtherEmitter.SampledAge
      )
     )
    ))
   &CameraOffset(Camera Offset Amount=&RandomRangeFloat(Minimum=500, Maximum=500))
   /SetVariables(Particles.SpriteSize=&Lerp_Vector2(
     A=<0 0>
     B=&Vector2DFromFloat(Value=&RandomRangeFloat(Minimum=2000, Maximum=2000))
     Alpha=&FloatFromCurve(
      FloatCurve=DataInterface() {propertyValues {Curve {keys [<at 12 0>,<at 13 1>,<at 17 1>,<at 17.1 0>]}}}
      CurveIndex=$Output.SampleParticlesFromOtherEmitter.SampledAge
     )
    )) {isSetParameters=true}
  ]
  particleUpdateScript [
   &ParticleState()
   &ScaleSpriteSize(Uniform Curve Sprite Scale=DataInterface() {propertyValues {Curve {keys [{interpMode=RCIM_Cubic, tangentMode=RCTM_User, time=0.8, value=1}, {interpMode=RCIM_Cubic, tangentMode=RCTM_User, time=1, value=0}]}}})
   &GravityForce()
   &Drag(Drag=2, Use Linear Drag=true, Use Rotational Drag=true)
   &SolveForcesAndVelocity(Clamp Velocity=false, Limit Acceleration=false)
   &SubUVAnimation(
    End Frame=50
    Start Frame:NiagaraInt32=&RandomRangeInt(Minimum=6, Maximum=12)
    SubUV Animation Mode:ENiagara_SubUVLookupMode=::NewEnumerator1
   )
   &ScaleColor(
    Scale Alpha=&FloatFromCurve(
     FloatCurve=DataInterface() {propertyValues {Curve {keys [
         {interpMode=RCIM_Cubic, tangentMode=RCTM_User, time=0, value=0}
         {interpMode=RCIM_Cubic, tangentMode=RCTM_User, time=0.25, value=1}
         {tangentMode=RCTM_User, time=1, value=0}
        ]}}}
     Scale Curve=1.5
    )
    Scale RGB=&MakeVectorFromLinearColorRGB(Color=<rgba 0 0 0 1>)
    ScaleA=true
    ScaleRGB=false
   )
   &ScaleColor(
    Scale Alpha=&FloatFromCurve(
     FloatCurve=DataInterface() {propertyValues {Curve {keys [
         {interpMode=RCIM_Cubic, tangentMode=RCTM_User, time=0, value=0}
         {interpMode=RCIM_Cubic, tangentMode=RCTM_User, time=0.5, value=1}
         {tangentMode=RCTM_User, time=1, value=0}
        ]}}}
     CurveIndex=$Particles.Random
    )
    ScaleA=true
    ScaleRGB=true
   ) {module=ScaleColor001}
   &ScaleColor(
    Scale Alpha=&FNTimeOfDayScalar(Night Scale=0.5, Morning Scale=1, Evening Scale=1)
    ScaleA=true
    ScaleRGB=true
   ) {module=ScaleColor002}
  ]
  rendererClasses [&NiagaraSpriteRendererProperties]
  renderers [{rendererClass=&NiagaraSpriteRendererProperties}]
 }
 Sparkles_V2 {
  simTarget=GPUComputeSim
  emitterUpdateScript [
   &EmitterState(Life Cycle Mode=::NewEnumerator1, Loop Duration=1, UseLoopDelay=false) {module=EmitterLifeCycle}
   &SpawnParticlesFromOtherEmitter(
    Attribute Reader=DataInterface() {propertyValues {EmitterBinding {bindingMode=Other, emitterName=WaveMesh}}}
    Spawn Rate=&MakeCustomFloatFromBool(
     Bool=&SetBoolByFloatComparison(A=$NPC.FortniteNPC.Main_Phase_Tracker, B=3, Comparison Type=::NewEnumerator4)
     True Float=2500
     False Float=1250
    )
    Spawn Rate Per Particle Cap=2500
    Use Maximum Age=false
    Use Minimum Age=false
   )
  ]
  particleSpawnScript [
   &InitializeParticle(
    Color=&RandomRangeLinearColor(Minimum=<rgba 0 0.8192 1 1>, Maximum=<rgba 0.6188 0 1 1>)
    Lifetime=&RandomRangeFloat(Minimum=0.5, Maximum=2)
    Sprite Size Mode=::NewEnumerator3
    Uniform Sprite Size=&RandomRangeFloat(Minimum=10, Maximum=20)
    UsePositionOffset=false
    Write Material Random=true
   ) {module=InitializeParticle001}
   &SampleParticlesFromOtherEmitter(
    Age Sampling:ENiagara_DirectReadSamplingMode=::NewEnumerator1
    Color Sampling=::NewEnumerator1
    Lifetime Sampling:ENiagara_DirectReadSamplingMode=::NewEnumerator1
    Position Sampling:ENiagara_DirectReadSamplingMode=::NewEnumerator1
   )
   /SetVariables(Particles.Random=&RandomRangeFloat(Minimum=0, Maximum=1)) {isSetParameters=true, module=SetVariables_#4}
   &ShapeLocation(
    Additional Quaternion Rotation=false
    Cylinder Distribution=::NewEnumerator2
    Cylinder Height=200
    Cylinder Radius=&RandomRangeFloat(Minimum=28000, Maximum=31000)
    MaskSpawnGroup=false
    Offset=&Multiply_VectorByFloat(Vector=$System.WindDirection, Float=-29000)
    Offset Coordinate Space=::World
    Offset Mode=::NewEnumerator0
    Rotation Mode=::NewEnumerator3
    Rotation Quaternion=&FindQuatBetweenTwoVectors(From Vector=$System.WindDirection, To Vector=<1 0 0>)
    Shape Origin=&Add_Vector(
     A=&MakeVector(
      X=0
      Y=0
      Z=&FloatFromCurve(
       FloatCurve=DataInterface() {propertyValues {Curve {keys [<at 18 250>,<at 20 -2000>]}}}
       CurveIndex=$Output.SampleParticlesFromOtherEmitter.SampledAge
      )
     )
     B=$Output.SampleParticlesFromOtherEmitter.SampledPosition
    )
    Shape Primitive=::NewEnumerator1
    U Position=&FloatFromCurve(FloatCurve=DataInterface() {propertyValues {Curve {keys [<at 0 0.14>,<at 1 0.36>]}}}, CurveIndex=$Particles.Random)
    UseOwnerScale=true
    V Position=&RandomRangeFloat(Minimum=0, Maximum=1)
   ) {module=ShapeLocation001}
   &AddVelocity(Velocity=$System.WindDirection, Velocity Speed Scale=500) {module=AddVelocity001}
   &Color(Color=&MultiplyLinearColor(
     Linear Color=&RandomRangeLinearColor(Minimum=<rgba 0 0.8192 1 1>, Maximum=<rgba 0.6188 0 1 1>)
     Scale Factor=&MakeVector4_FromFloat(Float=&FloatFromCurve(
       FloatCurve=DataInterface() {propertyValues {Curve {keys [<at 14.5 0>,<at 15.5 1>,<at 16.5 1>,<at 16.5 1>,<at 17.5 0>]}}}
       CurveIndex=$Output.SampleParticlesFromOtherEmitter.SampledAge
      ))
    ))
   /SetVariables(Particles.SpriteSize=&Lerp_Vector2(
     A=<0 0>
     B=&Vector2DFromFloat(Value=&RandomRangeFloat(Minimum=10, Maximum=20))
     Alpha=&MakeFloatFromVector4(
      Vector 4=&MakeVector4_FromFloat(Float=&FloatFromCurve(
        FloatCurve=DataInterface() {propertyValues {Curve {keys [<at 14.5 0>,<at 15.5 1>,<at 16.5 1>,<at 17.5 0>]}}}
        CurveIndex=$Output.SampleParticlesFromOtherEmitter.SampledAge
       ))
      Channel=::NewEnumerator0
     )
    )) {isSetParameters=true}
  ]
  particleUpdateScript [
   &ParticleState()
   &ScaleSpriteSize()
   &ScaleSpriteSize(Scale Factor=&Vector2DFromFloat(Value=&FloatFromCurve(FloatCurve=DataInterface() {propertyValues {Curve {keys [<at 0 1 RCIM_Cubic>, {arriveTangent=-1.6315, interpMode=RCIM_Cubic, leaveTangent=-1.6315, tangentMode=RCTM_User, time=1, value=0}]}}}))) {module=ScaleSpriteSize001}
   &CurlNoiseForce(Noise Quality / Cost=::NewEnumerator0, Noise Strength=500, Pan Noise=false)
   &GravityForce(Gravity=&RandomRangeVector(Minimum=<0 0 40>, Maximum=<0 0 160>))
   &Drag(
    Drag=&FloatFromCurve(
     FloatCurve=DataInterface() {propertyValues {Curve {keys [
         {arriveTangent=9.0725, interpMode=RCIM_Cubic, leaveTangent=9.0725, tangentMode=RCTM_User, time=0, value=0}
         {arriveTangent=-2.5249, interpMode=RCIM_Cubic, leaveTangent=-2.5249, tangentMode=RCTM_User, time=0.1408, value=0.9847}
         {arriveTangent=-0.0405, interpMode=RCIM_Cubic, leaveTangent=-0.0405, tangentMode=RCTM_User, time=1, value=0}
        ]}}}
     Scale Curve=9
    )
    Use Linear Drag=true
    Use Rotational Drag=true
   )
   &SolveForcesAndVelocity(Clamp Velocity=false, Limit Acceleration=false)
   &ScaleColor(
    Scale Alpha=&FloatFromCurve(FloatCurve=DataInterface() {propertyValues {Curve {keys [{interpMode=RCIM_Cubic, tangentMode=RCTM_User, time=0.5, value=1}]}}})
    Scale RGB=&VectorFromFloat(Value=&FloatFromCurve(FloatCurve=DataInterface() {propertyValues {Curve {keys [<at 0 50>,<at 0.11 1>]}}}))
    ScaleA=true
    ScaleRGB=true
   )
   /SetVariables(Particles.FlareScale:Vector2f=&Multiply_Vector2D_ByFloat(Float=10, Vector2D=$Particles.SpriteSize)) {isSetParameters=true}
   /SetVariables(Particles.FlareColor=&MultiplyLinearColor(Linear Color=$Particles.Color, Scale Factor=<1 1 1 0.25>)) {isSetParameters=true}
  ]
  rendererClasses [&NiagaraSpriteRendererProperties]
  renderers [
   {rendererClass=&NiagaraSpriteRendererProperties}
   {rendererClass=&NiagaraSpriteRendererProperties, rendererIndex=1}
  ]
 }
 WaveMesh {
  simTarget=GPUComputeSim
  emitterUpdateScript [
   &EmitterState(Life Cycle Mode=::NewEnumerator1, Loop Duration=5.2174, UseLoopDelay=false)
   &SpawnBurst_Instantaneous(Use Spawn Probability=false, UseLoopCountLimit=false)
  ]
  particleSpawnScript [
   &InitializeParticle(
    Lifetime=20
    Mesh Scale=<2 2 0.25>
    Mesh Scale Mode=::NewEnumerator4
    Sprite Rotation Mode=::NewEnumerator2
    Sprite Size Mode=::NewEnumerator3
    Uniform Sprite Size=&RandomRangeFloat(Minimum=2000, Maximum=2000)
    UsePositionOffset=false
    Write Material Random=true
   )
   /SetVariables(Particles.Random=&RandomRangeFloat(Minimum=0, Maximum=1)) {isSetParameters=true, module=SetVariables_#5}
   &CameraOffset(Camera Offset Amount=75)
   &AddVelocity(Rotation Coordinate Space=::World, Velocity=$System.WindDirection, Velocity Speed Scale=2000)
   &InitialMeshOrientation(Mesh Axis To Orient=<1 1 0>, Mesh Orientation Mode=::NewEnumerator2, Use Rotation Vector=false)
  ]
  particleUpdateScript [
   &ParticleState()
   &ScaleMeshSize(Scale Factor=&MakeVector(
     X=1
     Y=&FloatFromCurve(FloatCurve=DataInterface() {propertyValues {Curve {keys [
          {tangentMode=RCTM_User, time=0, value=1}
          {arriveTangent=-0.0405, interpMode=RCIM_Cubic, leaveTangent=-0.0405, tangentMode=RCTM_User, time=0.7, value=1}
          {interpMode=RCIM_Cubic, tangentMode=RCTM_User, time=0.9, value=0.5}
         ]}}})
     Z=&FloatFromCurve(FloatCurve=DataInterface() {propertyValues {Curve {keys [{interpMode=RCIM_Cubic, tangentMode=RCTM_User, time=0.8, value=1}, {interpMode=RCIM_Cubic, tangentMode=RCTM_User, time=1, value=0}]}}})
    ))
   /SetVariables(Particles.Radius:NiagaraFloat=&Lerp_Float(
     A=0
     B=$System.WaterLength
     Alpha=&FloatFromCurve(FloatCurve=DataInterface() {propertyValues {Curve {keys [
          {arriveTangent=4.1785, interpMode=RCIM_Cubic, leaveTangent=4.1785, tangentMode=RCTM_User, time=0, value=0}
          {arriveTangent=1.496, interpMode=RCIM_Cubic, leaveTangent=1.496, tangentMode=RCTM_User, time=0.2042, value=0.6228}
          {
           arriveTangentWeight=0.4724
           interpMode=RCIM_Cubic
           leaveTangentWeight=0.4724
           tangentMode=RCTM_User
           tangentWeightMode=RCTWM_WeightedBoth
           time=0.97
           value=1
          }
         ]}}})
    )) {isSetParameters=true, module=SetVariables_#6}
   &ShapeLocation(
    Additional Quaternion Rotation=false
    Additional Yaw / Pitch / Roll=<0 0.5 0>
    Cylinder Distribution=::NewEnumerator2
    Cylinder Height=0
    Cylinder Radius=50000
    MaskSpawnGroup=false
    Non Uniform Scale=<3 3 1>
    Offset=&Multiply_VectorByFloat(Vector=$System.WindDirection, Float=$Particles.Radius)
    Offset Coordinate Space=::World
    Offset Mode=::NewEnumerator0
    Rotation Mode=::NewEnumerator3
    Rotation Quaternion=&FindQuatBetweenTwoVectors(From Vector=$System.WindDirection, To Vector=<1 0 0>)
    Shape Origin=&AddVectorToPosition(
     Position=/Niagara/DynamicInputs/Helpers/SimulationPosition()
     Vector=&MakeVector(
      X=0
      Y=0
      Z=&FloatFromCurve(FloatCurve=DataInterface() {propertyValues {Curve {keys [<at 0.95 -4000>]}}})
     )
    )
    Shape Primitive=::NewEnumerator1
    U Position=&Multiply_Float(A=0.24, B=1)
    UseOwnerScale=false
   ) {module=ShapeLocation001}
   &SolveForcesAndVelocity(Clamp Velocity=false, Limit Acceleration=false) {module=SolveForcesAndVelocity001}
   &ScaleColor(
    Scale Alpha=&FloatFromCurve(FloatCurve=DataInterface() {propertyValues {Curve {keys [{interpMode=RCIM_Cubic, tangentMode=RCTM_User, time=0.8, value=1}, <at 1 0 RCIM_Cubic>]}}})
    ScaleA=true
    ScaleRGB=true
   ) {module=ScaleColor003}
   &DynamicMaterialParameters(
    Index 0 Param 1=&FloatFromCurve(FloatCurve=DataInterface() {propertyValues {Curve {keys [{interpMode=RCIM_Cubic, tangentMode=RCTM_User, time=0.8, value=0.5}, {interpMode=RCIM_Cubic, tangentMode=RCTM_User, time=0.9, value=1}]}}})
    Index 0 Param 2=&FloatFromCurve(FloatCurve=DataInterface() {propertyValues {Curve {keys [{interpMode=RCIM_Cubic, tangentMode=RCTM_User, time=0.75, value=0}, {interpMode=RCIM_Cubic, tangentMode=RCTM_User, time=0.9, value=1}]}}})
    Index 0 Param 3=&FloatFromCurve(FloatCurve=DataInterface() {propertyValues {Curve {keys [{interpMode=RCIM_Cubic, tangentMode=RCTM_User, time=0.8, value=0}, {interpMode=RCIM_Cubic, tangentMode=RCTM_User, time=1, value=1}]}}})
    Module.Index0Param0Write=true
    Module.Index0Param1Write=true
    Module.Index0Param2Write=true
    Module.Index0Param3Write=true
   )
  ]
  rendererClasses [&NiagaraMeshRendererProperties]
  renderers [{rendererClass=&NiagaraMeshRendererProperties, rendererIndex=0}]
 }
 WaveMesh001 {
  emitterUpdateScript [
   &EmitterState(Life Cycle Mode=::NewEnumerator1, Loop Duration=5.2174, UseLoopDelay=false)
   &SpawnBurst_Instantaneous(Use Spawn Probability=false, UseLoopCountLimit=false)
  ]
  particleSpawnScript [
   &InitializeParticle(
    Lifetime=20
    Mesh Scale=<2 2 0.25>
    Mesh Scale Mode=::NewEnumerator4
    Sprite Rotation Mode=::NewEnumerator2
    Sprite Size Mode=::NewEnumerator3
    Uniform Sprite Size=&RandomRangeFloat(Minimum=2000, Maximum=2000)
    UsePositionOffset=false
    Write Material Random=true
   )
   /SetVariables(Particles.Random=&RandomRangeFloat(Minimum=0, Maximum=1)) {isSetParameters=true, module=SetVariables_#5}
   &CameraOffset(Camera Offset Amount=75)
   &AddVelocity(Rotation Coordinate Space=::World, Velocity=$System.WindDirection, Velocity Speed Scale=2000)
   &InitialMeshOrientation(Mesh Axis To Orient=<1 1 0>, Mesh Orientation Mode=::NewEnumerator2, Use Rotation Vector=false)
  ]
  particleUpdateScript [
   &ParticleState()
   &ScaleMeshSize(Scale Factor=&MakeVector(
     X=1
     Y=&FloatFromCurve(FloatCurve=DataInterface() {propertyValues {Curve {keys [
          {tangentMode=RCTM_User, time=0, value=1}
          {arriveTangent=-0.0405, interpMode=RCIM_Cubic, leaveTangent=-0.0405, tangentMode=RCTM_User, time=0.7, value=1}
          {interpMode=RCIM_Cubic, tangentMode=RCTM_User, time=0.9, value=0.5}
         ]}}})
     Z=&FloatFromCurve(FloatCurve=DataInterface() {propertyValues {Curve {keys [{interpMode=RCIM_Cubic, tangentMode=RCTM_User, time=0.8, value=1}, {interpMode=RCIM_Cubic, tangentMode=RCTM_User, time=1, value=0}]}}})
    ))
   /SetVariables(Particles.Radius:NiagaraFloat=&Lerp_Float(
     A=0
     B=$System.WaterLength
     Alpha=&FloatFromCurve(FloatCurve=DataInterface() {propertyValues {Curve {keys [
          {arriveTangent=4.1785, interpMode=RCIM_Cubic, leaveTangent=4.1785, tangentMode=RCTM_User, time=0, value=0}
          {arriveTangent=1.496, interpMode=RCIM_Cubic, leaveTangent=1.496, tangentMode=RCTM_User, time=0.2042, value=0.6228}
          {
           arriveTangentWeight=0.4724
           interpMode=RCIM_Cubic
           leaveTangentWeight=0.4724
           tangentMode=RCTM_User
           tangentWeightMode=RCTWM_WeightedBoth
           time=0.97
           value=1
          }
         ]}}})
    )) {isSetParameters=true, module=SetVariables_#6}
   &ShapeLocation(
    Additional Quaternion Rotation=false
    Additional Yaw / Pitch / Roll=<0 0.5 0>
    Cylinder Distribution=::NewEnumerator2
    Cylinder Height=0
    Cylinder Radius=50000
    MaskSpawnGroup=false
    Non Uniform Scale=<3 3 1>
    Offset=&Multiply_VectorByFloat(Vector=$System.WindDirection, Float=$Particles.Radius)
    Offset Coordinate Space=::World
    Offset Mode=::NewEnumerator0
    Rotation Mode=::NewEnumerator3
    Rotation Quaternion=&FindQuatBetweenTwoVectors(From Vector=$System.WindDirection, To Vector=<1 0 0>)
    Shape Origin=&AddVectorToPosition(
     Position=&SimulationPosition()
     Vector=&MakeVector(
      X=0
      Y=0
      Z=&FloatFromCurve(FloatCurve=DataInterface() {propertyValues {Curve {keys [<at 0.95 -4000>]}}})
     )
    )
    Shape Primitive=::NewEnumerator1
    U Position=&Multiply_Float(A=0.24, B=1)
    UseOwnerScale=false
   ) {module=ShapeLocation001}
   &SolveForcesAndVelocity(Clamp Velocity=false, Limit Acceleration=false) {module=SolveForcesAndVelocity001}
   /SetVariables(Particles.NewColor:LinearColor=&MultiplyLinearColor(
     Linear Color=&MultiplyLinearColor(
      Linear Color=$Particles.Color
      Scale Factor=&MakeVector4_FromLinearColor(LinearColor=&ColorFromCurve(
        ColorCurve=DataInterface() {propertyValues {
          AlphaCurve {keys [<at 0 1 RCIM_Cubic>,<at 1 1 RCIM_Cubic>]}
          BlueCurve {keys [
            {arriveTangent=1.3581, interpMode=RCIM_Cubic, time=0, value=0.15}
            <at 0.0736 0.25>
            <at 0.1163 0.2448>
            <at 0.15 0.0328>
            <at 0.2091 0.25>
            <at 0.25 0.25>
            <at 0.455 0.5>
            <at 0.6597 0.5>
            <at 0.7304 0.25>
            <at 0.8114 0.25>
            <at 0.8468 0.0741>
            <at 0.8983 0.0649>
            <at 0.9248 0.15>
            <at 1 0.15 RCIM_Cubic>
           ]}
          GreenCurve {keys [
            {arriveTangent=1.3581, interpMode=RCIM_Cubic, time=0, value=0.15}
            <at 0.0736 0.25>
            <at 0.1163 0.2448>
            <at 0.15 0.0477>
            <at 0.2091 0.25>
            <at 0.25 0.25>
            <at 0.455 0.5>
            <at 0.6597 0.5>
            <at 0.7304 0.25>
            <at 0.8114 0.25>
            <at 0.8468 0.0741>
            <at 0.8983 0.0649>
            <at 0.9248 0.15>
            <at 1 0.15 RCIM_Cubic>
           ]}
          RedCurve {keys [
            {arriveTangent=1.3581, interpMode=RCIM_Cubic, time=0, value=0.15}
            <at 0.0736 0.25>
            <at 0.1163 0.2448>
            <at 0.15 0.1615>
            <at 0.2091 0.25>
            <at 0.25 0.25>
            <at 0.455 0.5>
            <at 0.6597 0.5>
            <at 0.7304 0.25>
            <at 0.8114 0.25>
            <at 0.8468 0.1094>
            <at 0.8983 0.1094>
            <at 0.9248 0.15>
            <at 1 0.15 RCIM_Cubic>
           ]}
         }}
        CurveIndex=Linked() {linkedVariable {name=NPC.FortniteNPC.TimeOfDay_0-1, type=NiagaraFloat}}
       ))
     )
     Scale Factor=<1 1 1 1>
    )) {enabled=false, isSetParameters=true}
   &ScaleColor(
    Scale Alpha=&FloatFromCurve(FloatCurve=DataInterface() {propertyValues {Curve {keys [{interpMode=RCIM_Cubic, tangentMode=RCTM_User, time=0.75, value=1}, <at 0.9 0 RCIM_Cubic>]}}})
    ScaleA=true
    ScaleRGB=true
   ) {module=ScaleColor003}
   &DynamicMaterialParameters(
    Index 0 Param 1=&FloatFromCurve(FloatCurve=DataInterface() {propertyValues {Curve {keys [{interpMode=RCIM_Cubic, tangentMode=RCTM_User, time=0.8, value=0.5}, {interpMode=RCIM_Cubic, tangentMode=RCTM_User, time=0.9, value=1}]}}})
    Index 0 Param 2=&FloatFromCurve(FloatCurve=DataInterface() {propertyValues {Curve {keys [{interpMode=RCIM_Cubic, tangentMode=RCTM_User, time=0.75, value=0}, {interpMode=RCIM_Cubic, tangentMode=RCTM_User, time=0.9, value=1}]}}})
    Index 0 Param 3=&FloatFromCurve(FloatCurve=DataInterface() {propertyValues {Curve {keys [{interpMode=RCIM_Cubic, tangentMode=RCTM_User, time=0.8, value=0}, {interpMode=RCIM_Cubic, tangentMode=RCTM_User, time=0.9, value=1}]}}})
    Module.Index0Param0Write=true
    Module.Index0Param1Write=true
    Module.Index0Param2Write=true
    Module.Index0Param3Write=true
   )
  ]
  rendererClasses [&NiagaraMeshRendererProperties]
  renderers [{rendererClass=&NiagaraMeshRendererProperties, rendererIndex=0}]
 }
}
summary {
 systemName=NS_KelpLinen_Shockwave
 emitters [
  {emitterName=WaveMesh, simTarget=GPUComputeSim, rendererClasses [&NiagaraMeshRendererProperties]}
  {emitterName=Sparkles_V2, simTarget=GPUComputeSim, rendererClasses [&NiagaraSpriteRendererProperties]}
  {emitterName=Smoke_V2, simTarget=GPUComputeSim, rendererClasses [&NiagaraSpriteRendererProperties]}
  {emitterName=WaveMesh001, rendererClasses [&NiagaraMeshRendererProperties]}
 ]
 userVariables [
  User.RadialShockwave_On {type=NiagaraBool, defaultValue=true}
  User.SkeletalMesh {
   type=NiagaraDataInterfaceSkeletalMesh
   defaultValue {
    struct=NiagaraExt_VariableValue_DataInterface
    value {dataInterface=&self:NiagaraDataInterfaceSkeletalMesh_0}
   }
  }
 ]
}
